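; Creative player spawner device: a placeable pad where players spawn in a minigame or island, filtered by team, class and priority.
; The server pushes options onto a child FortPlayerStartCreative, while clients draw team-colored plate, async-loaded edit-mode hologram, enemy-range visual and audio.
; SpawnPlayer reserves the pad by player name and top priority with a timeout, optionally killing living players to respawn them.
; Minigame spawn callbacks broadcast spawned events for claiming players.
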(event EventBeginPlay
  ;[0 1 0.1154] Initialization
  (|CallParentFunction)
  (CallFunction|InitializeAfterRestorefromSave))

(event EventPostFinishSpawning
  ;[0 1 0.1154] Initialization
  (|CallParentFunction)
  (CallFunction|InitializeAfterRestorefromSave))

(event Custom|Initialize%20Device
  ;[0 1 0.1154] Initialization
  (|CallParentFunction)
  (Class|CreativeEnabledComponent|Initialize (Variables|Default|GetCreative_Enabled_Component) (Variables|Default|GetFortMinigameProgress) (Variables|UserOptions|GetEnabledDuringPhase))
  (CallFunction|ReinitializeSpawnSettings)
  ;; SetupEditorOnlyVisibility
  (Initialization|InitializeWithSceneComponents (Variables|Default|GetVisibleInEditorOnlyComponent) (Utilities|Array|MakeArray (Variables|Default|GetPlayerMesh) (Variables|Default|GetTeamNumberDisplay)) (Utilities|Map|MakeMap) "NewEnumerator1"))

(event EventOnWorldReady
  ;[0 1 0.1154] Initialization
  (|CallParentFunction)
  (CallFunction|ReinitializeSpawnSettings))

(event Custom|Reinitialize%20Spawn%20Settings
  ;[0 1 0.1154] Initialization
  (Networking|CanExecuteCosmeticEvents
    (:True
      (CallFunction|InitializeTeamAssignmentDisplay)
      (CallFunction|UpdateAudio)
      (CallFunction|UpdateHologram (not (Default|IsinPlayMode)))
      (CallFunction|InitializeEnemyRangeVisualization))
    (:False
      (Networking|SwitchHasAuthority
        (:Authority
          (CallFunction|UpdatePriorityGroup)
          (CallFunction|UpdateUseasIslandStart)
          (CallFunction|UpdateTeamSetting)
          (UserOptions|UpdateClassSetting)
          (CallFunction|UpdateEnemyRangeCheck))))))

(event OnAddedtoMinigame(FortMinigameProgress)
  (params (Minigame :type Fort%20Minigame%20Object%20Reference))
  ;[0 1 0.1154] Initialization
  (CallFunction|BindEventtoMinigamePlayerSpawn)
  (CallFunction|BindEventtoMinigameTeamConfigUpdated))

(event PostMinigameStartedDelegate(FortMinigamePlayerStart)
  ;[0 0.6223 1 0.25] State Change Logic
  (Networking|SwitchHasAuthority
    (:Authority
      (bind _controllertransmitingongamestart (Variables|Default|GetControllerTransmitingOnGameStart)
        (:then
          (CallFunction|TransmitonPlayerSpawned _controllertransmitingongamestart))
        (:else)))))

(event MinigameStateChangedDelegate(FortMinigameProgress)
  (params (Minigame :type Fort%20Minigame%20Object%20Reference) (NewMinigameState :type EFortMinigameState%20Enum :default "PreGame"))
  ;[0 0.6223 1 0.25] State Change Logic
  (CallFunction|UpdateUseasIslandStart)
  (switch Utilities|FlowControl|Switch|SwitchonEFortMinigameState NewMinigameState
    (:WaitingForCameras
      (Audio|PushSoundMixModifier "/Game/Sounds/Creative/MixModifiers/Creative_InGameStart_MixMod.Creative_InGameStart_MixMod"))))

(event Custom|Enable%20Event
  (params (Instigator :type Controller%20Object%20Reference))
  ;[0.6648 0 1] Gameplay Functions
  (Class|CreativeEnabledComponent|SetEnabled (Variables|Default|GetCreative_Enabled_Component) true))

(event Custom|Disable%20Event
  (params (Instigator :type Controller%20Object%20Reference))
  ;[0.6648 0 1] Gameplay Functions
  (Class|CreativeEnabledComponent|SetEnabled (Variables|Default|GetCreative_Enabled_Component) false))

(event OnEnabledStateChanged(Creative_Enabled_Component)
  (params (Enabled :type Boolean))
  ;[0 0.6223 1 0.25] State Change Logic
  (CallFunction|UpdateIsEnabled Enabled)
  (CallFunction|ReinitializeSpawnSettings))

(event Custom|OnMinigamePlayerSpawned
  (params (RespawningPlayer :type Controller%20Object%20Reference))
  (bind _asfort_player_state (Utilities|Casting|CastToFortPlayerState (Variables|/Script/Engine.Controller|GetPlayerState RespawningPlayer)))
  (Utilities|IsValid _asfort_player_state
    (:Is%20Valid
      (bind _isinprogress (Utilities|IsMinigameStateinProgress)
        (:ValidMinigame
          (if _isinprogress
            (CallFunction|TransmitonPlayerSpawned _asfort_player_state)
            (else
              (CallFunction|ValidateandCacheTransmitData _asfort_player_state))))))))

(event Custom|SpawnPlayer%20Event
  (params (Instigator :type Controller%20Object%20Reference))
  ;[0.6648 0 1] Gameplay Functions
  (Networking|SwitchHasAuthority
    (:Authority
      (bind _asfort_player_state_athena (Utilities|Casting|CastToFortPlayerStateAthena (Variables|/Script/Engine.Controller|GetPlayerState Instigator))
        (:then
          (Utilities|IsValid (Class|FortPlayerState|FindFNEPlayerComponent _asfort_player_state_athena)
            (:Is%20Valid
              (if (not (Minigame|IsPlayerTeleportingorRespawningforGameplay (Class|FortPlayerState|FindFNEPlayerComponent _asfort_player_state_athena)))
                (bind _bisreserved (CallFunction|ReserveSpawnerforPlayer _asfort_player_state_athena))
                (if _bisreserved
                  (bind _asfort_player_controller_athena (Utilities|Casting|CastToFortPlayerControllerAthena (PlayerState|GetPlayerController _asfort_player_state_athena))
                    (:then
                      (Variables|Default|SetRespawningPlayerController _asfort_player_controller_athena)
                      (Utilities|IsValid (Game|Player|GetPlayerPawn _asfort_player_controller_athena)
                        (:Is%20Valid
                          (if (not (FortPawn|IsDead (Game|Player|GetPlayerPawn _asfort_player_controller_athena)))
                            (if (Variables|UserOptions|GetbRespawnAlivePlayers)
                              ;;[0 0.9666 1 0.5] Do not kill the Pawn if it isn't done possessing
                              (if (Class|FortPlayerController|IsPawnAcknowlodged (Variables|Default|GetRespawningPlayerController))
                                (FortPawn|ForceKill (Game|Player|GetPlayerPawn (Variables|Default|GetRespawningPlayerController)) "(TagName=\"Gameplay.Effect.InstantDeath\")" (Variables|Default|GetRespawningPlayerController) (Game|Player|GetPlayerPawn (Variables|Default|GetRespawningPlayerController)))
                                (else
                                  _))
                              (Class|FortMinigamePlayerSpawnComponent|SetRespawnDataandRespawn (Minigame|GetPlayerSpawnComponent (Variables|/Script/FortniteGame.FortMinigameLogicComponent|GetCurrentMinigame (Variables|Default|GetFortMinigameProgress))) (Variables|/Script/Engine.Controller|GetPlayerState (Variables|Default|GetRespawningPlayerController))))
                            (else
                              _)))
                        (:Is%20Not%20Valid
                          _)))
                    (:CastFailed
                      _))
                  (else
                    _))
                (CallFunction|TransmitonSpawnFailed 0)
                (else
                  _)))
            (:Is%20Not%20Valid
              _)))
        (:CastFailed
          _)))))

(event OnVisibilityChanged(VisibleInGameComponent)
  (params (Visible :type Boolean))
  ;[0.3118 0.5472 1 0.25] Turn off Collision when invisible
  (Collision|SetCollisionProfileName (Variables|Debug|GetStaticMeshComponent) (select Visible "FortDynamicMeshPhysics" "OverlapAll") false :!class /Script/Engine.PrimitiveComponent)
  (CallFunction|UpdateAudio))

(event Custom|OnMinigameTeamConfigUpdated
  ;[0.3118 0.5472 1 0.25] Handle Team Config Update
  (if (not (Default|IsinPlayMode))
    (CallFunction|UpdateHologramVisuals)))

(event EventOnPlayModeChanged
  (params (Minigame :type Fort%20Minigame%20Object%20Reference) (bIsInPlayMode :type Boolean))
  ;[0 0.7015 1 0.25] We only want to be checking for nearby players during edit mode, and only on the client.
  (|CallParentFunction Minigame bIsInPlayMode)
  (CallFunction|UpdateHologram (not bIsInPlayMode)))

(event Custom|Update%20Hologram
  (params (Show :type Boolean))
  (bind _playermesh (Variables|Default|GetPlayerMesh))
  (bind _playermesh_2 (Variables|Default|GetPlayerMesh))
  ;[1 0.6509 0 0.25] If we're in edit mode, load all of our hologram data. Else don't.
  (Networking|CanExecuteCosmeticEvents
    (:True
      (if Show
        (bind _loaded_object (Creative|Loading|AsyncLoadEditorSafe (Variables|Default|GetHologram%20Mesh))
          (:Loaded
            (bind _asskeletal_mesh (Utilities|Casting|CastToSkeletalMesh _loaded_object))
            (Components|SkeletalMesh|SetSkeletalMeshAsset _playermesh_2 _asskeletal_mesh)
            (bind _loaded_object_2 (Creative|Loading|AsyncLoadEditorSafe (Variables|Default|GetHologram%20Animation))
              (:Loaded
                (bind _asanim_sequence_base (Utilities|Casting|CastToAnimSequenceBase _loaded_object_2))
                (Components|Animation|SetAnimation _playermesh_2 _asanim_sequence_base)
                (bind _loaded_object_3 (Creative|Loading|AsyncLoadEditorSafe (Variables|Default|GetHologram%20Material))
                  (:Loaded
                    (bind _asmaterial_interface (Utilities|Casting|CastToMaterialInterface _loaded_object_3))
                    (Rendering|Material|SetMaterial _playermesh_2 0 _asmaterial_interface)
                    (Rendering|Material|SetMaterial _playermesh_2 1 _asmaterial_interface)
                    (CallFunction|UpdateHologramVisuals)))))))
        (else
          (Components|SkeletalMesh|SetSkeletalMeshAsset _playermesh 0)
          (Components|Animation|SetAnimation _playermesh 0))))))

(event OnMeshandMaterialsLoaded(TeamNumberDisplay)
  (CallFunction|InitializeTeamAssignmentDisplay))

(event OnMeshandMaterialsLoaded(EnemyRangeVisualization)
  (CallFunction|InitializeEnemyRangeVisualization))

(event OnRemovedfromMinigame(FortMinigameProgress)
  (params (Minigame :type Fort%20Minigame%20Object%20Reference))
  ;[1 0.01591 0 0.25] Cleanup on Destruction
  (CallFunction|UnbindEventtoMinigamePlayerSpawn)
  (CallFunction|UnbindEventtoMinigameTeamConfigUpdated))

(fn BindEventToMinigamePlayerSpawn
  (Networking|SwitchHasAuthority
    (:Authority
      (bind _currentminigame (Variables|/Script/FortniteGame.FortMinigameLogicComponent|GetCurrentMinigame (Variables|Default|GetFortMinigameProgress))
        (:then
          (Minigame|BindEventtoOnMinigamePlayerorBotSpawned _currentminigame (EventDispatchers|CreateEvent :!event OnMinigamePlayerSpawned)))
        (:else)))))

(fn BindEventToMinigameTeamConfigUpdated
  (Networking|CanExecuteCosmeticEvents
    (:True
      (bind _currentminigame (Variables|/Script/FortniteGame.FortMinigameLogicComponent|GetCurrentMinigame (Variables|Default|GetFortMinigameProgress))
        (:then
          (Minigame|BindEventtoMinigameTeamConfigUpdated _currentminigame (EventDispatchers|CreateEvent :!event OnMinigameTeamConfigUpdated)))
        (:else)))))

(fn ClearReserveSpawn
  (bind _reservedspawnpadtimer (Variables|Default|GetReservedSpawnPadTimer))
  (CallFunction|UpdatePriorityGroup)
  (bind _asfort_player_start_creative (Utilities|Casting|CastToFortPlayerStartCreative (Variables|/Script/Engine.ChildActorComponent|GetChildActor (Variables|Default|GetPlayerStartCreative))))
  (Variables|/Script/FortniteGame.FortPlayerStartCreative|SetLockedPlayerNameSubstring "" _asfort_player_start_creative)
  (Variables|Default|SetbIsReserved false)
  (Variables|Default|SetRespawningPlayerController 0)
  (Variables|Default|SetRespawnTimeDelay (Variables|Default|GetRespawnTimeDelta))
  (if (Utilities|Time|IsValidTimerHandle _reservedspawnpadtimer)
    (Utilities|Time|ClearandInvalidateTimerbyHandle _reservedspawnpadtimer)))

(fn GetCanSpawnPlayer
  (LocalVariables|SetCreativeRoyaleSpawning false)
  (bind _returnvalue (Minigame|GetMinigameforActor self))
  (Utilities|IsValid _returnvalue
    (:Is%20Valid
      (Utilities|IsValid (Minigame|GetPlayerSpawnComponent _returnvalue)
        (:Is%20Valid
          (bind _array_element (Utilities|Array|ForEachLoopwithBreak (Class|FortMinigamePlayerSpawnComponent|GetPlayerSpawnOverrideDevices (Minigame|GetPlayerSpawnComponent _returnvalue)))
            (:LoopBody
              ;[0.2864 1 0.5993] On Cast Success it means we are in Creative Royale, as the array should only have one asset, the deployment console. If true, check to see if we have the correct spawn method enabled.
              (bind _ascreative_spawn_override_interface (Utilities|Casting|CastToCreativeSpawnOverrideInterface _array_element))
              (LocalVariables|SetCreativeRoyaleSpawning true)
              (bind _output_get (LocalVariables|SetCanSpawnOnPadInCR (Utilities|Enum|Equal(Enum) "Spawner" (Class|CreativeSpawnOverrideInterface|GetPlayerSpawnMethod _ascreative_spawn_override_interface))))
              (return _output_get))
            (:Completed
              ;[1 0.8831 0.2224] Either we are not in Creative Royale, and it's safe to spawn, or we are in Creative Royale and we allow spawning. If not, return false
              (return true))))))))

(fn GetEnemyVisualizationMID
  (return (Material|GetMID (Variables|Default|GetEnemyRangeVisualization))))

(fn GetHologramMID
  (bind _playermesh (Variables|Default|GetPlayerMesh))
  (bind _hologrammid (Variables|Default|GetHologramMID))
  (Material|ReattachOrCreateMID _playermesh _hologrammid)
  (Rendering|Material|SetMaterial _playermesh 1 _hologrammid)
  (return _hologrammid))

(fn GetVisualComponents
  (return (Utilities|Array|MakeArray (Variables|Debug|GetStaticMeshComponent))))

(fn InitializeEnemyRangeVisualization
  (bind _enemyrangevisualization (Variables|Default|GetEnemyRangeVisualization))
  (Transformation|SetWorldScale3D (Variables|Default|GetEnemyRangeVisualization) (Math|Vector|SelectVector "0.010000,0.010000,0.010000" (Math|Conversions|ToVector(Float) (* (Variables|UserOptions|GetEnemyRangeCheck) 2.0)) (Utilities|Enum|Equal(Enum) (Variables|UserOptions|GetDisplayEnemyRange) "NewEnumerator0")))
  (Rendering|SetVisibility _enemyrangevisualization (and (Utilities|Enum|NotEqual(Enum) (Variables|UserOptions|GetDisplayEnemyRange) "NewEnumerator0") (> (Variables|UserOptions|GetEnemyRangeCheck) 0)) false)
  (Rendering|Material|SetScalarParameterValue (Material|GetMID _enemyrangevisualization) "FadeDistance" (Utilities|Select :Index (Variables|UserOptions|GetDisplayEnemyRange) :NewEnumerator0 0.0 :NewEnumerator1 -1.0 :NewEnumerator2 (Variables|Default|GetEnemyVisualizationRange)) :!class /Script/Engine.MaterialInstanceDynamic))

(fn InitializeAfterRestoreFromSave
  (bind _fortminigameplayerstart (Variables|Default|GetFortMinigamePlayerStart))
  (bind _asfort_player_start_creative (Utilities|Casting|CastToFortPlayerStartCreative (Variables|/Script/Engine.ChildActorComponent|GetChildActor (Variables|Default|GetPlayerStartCreative))))
  (CallFunction|SetSpawnerRotation)
  (Variables|/Script/FortniteGame.FortMinigamePlayerStartComponent|SetPlayerStarts (Utilities|Array|MakeArray _asfort_player_start_creative) _fortminigameplayerstart)
  (Variables|/Script/FortniteGame.FortPlayerStartCreative|SetPlayerStartTags (Variables|/Script/FortniteGame.FortMinigameLogicComponent|GetMinigameTags _fortminigameplayerstart) _asfort_player_start_creative))

(fn InitializeTeamAssignmentDisplay
  (bind _mid_team_assignment (Variables|Default|GetMID%20Team%20Assignment))
  (bind _teamnumberdisplay (Variables|Default|GetTeamNumberDisplay))
  (Variables|/Script/Engine.StaticMeshComponent|GetStaticMesh _teamnumberdisplay
    (:then
      (Material|ReattachOrCreateMID _teamnumberdisplay _mid_team_assignment)
      (Rendering|Material|SetScalarParameterValue _mid_team_assignment "Score" (Material|TranslateTeamValuetoMaterialScoreParam) :!class /Script/Engine.MaterialInstanceDynamic)
      (if (Editor|InEditor)
        (Rendering|Material|SetScalarParameterValue _mid_team_assignment "FadeDistance" (Variables|Default|GetHologramFadeDistanceEditor) :!class /Script/Engine.MaterialInstanceDynamic) ; In editor, we care about camera intersecting.\nIn Game, we care about character intersecting.
      ))
    (:else)) ; May initialize before Static Mesh\nhas loaded, so check first.
)

(fn IsClaimedBy
  (params (PlayerState :type Fort%20Player%20State%20Object%20Reference))
  (Utilities|IsValid PlayerState
    (:Is%20Valid
      (bind _asfort_player_start_creative (Utilities|Casting|CastToFortPlayerStartCreative (Variables|/Script/Engine.ChildActorComponent|GetChildActor (Variables|Default|GetPlayerStartCreative)))
        (:then
          (return (Class|FortPlayerStartCreative|IsClaimedbyPlayer _asfort_player_start_creative (PlayerState|GetUniqueID PlayerState))))
        (:CastFailed
          _)))
    (:Is%20Not%20Valid
      _))
  (return false))

(fn IsInsideSameVolume
  (params (PlayerState :type Player%20State%20Object%20Reference))
  (Utilities|IsValid PlayerState
    (:Is%20Valid
      (Utilities|IsActorRelevanttoMe PlayerState
        (:IsRelevant
          (return true))
        (:NotRelevant
          _)))
    (:Is%20Not%20Valid
      _))
  (return false))

(fn IsTargetableByLinearMover
  (return true))

(fn IsVisibleDuringMinigame
  (bind _output_get (Variables|Default|SetbIsVisibleDuringMinigame (Variables|UserOptions|Basic|GetVisibleInGame)))
  (return _output_get))

(fn MigrateLegacyUserOptions
  (bind (_enabledindex _prioritygroupsetting) (Class|GetClassDefaults "/CRD_PlayerSpawn/BP_Creative_Player_Spawner_Prop.BP_Creative_Player_Spawner_Prop_C"))
  (bind _prioritygroupsetting_2 (Variables|Legacy|GetPriorityGroupSetting))
  (|CallParentFunction)
  (Utilities|FlowControl|Sequence
    (:then_0
      ;[1 0.421 0.7323 0.25] Enabled int -> enum conversion (+1, since the int had a negative value).
      ; Priority Group split into separate Integer and Override options (to show as a bool with int in UEFN, and a Do Not Override with int in Creative)
      (Utilities|FlowControl|Sequence
        (:then_0
          (if (!= (Variables|Legacy|GetEnabledIndex) _enabledindex)
            (Variables|UserOptions|SetEnabledDuringPhase (Utilities|Enum|BytetoEnumEEnabledOnPhase (Math|Conversions|ToByte(Integer) (+ (Variables|Legacy|GetEnabledIndex) 1)))) ;     Changing Int to Enum.\nInteger had a -1 option, so we're \nshifting everything +1 to make it\n    match our new enum
            (Variables|Legacy|SetEnabledIndex _enabledindex) ; Set migrated property to \ndefault so won't migrate again.
          ))
        (:then_1
          (if (!= _prioritygroupsetting _prioritygroupsetting_2)
            (Variables|UserOptions|SetPriorityGroup_Override (> _prioritygroupsetting_2 0)) ; 0 value was 'Use Default' previously
            (Variables|UserOptions|SetPriorityGroup _prioritygroupsetting_2)
            (Variables|Legacy|SetPriorityGroupSetting _prioritygroupsetting) ; Set migrated property to \ndefault so won't migrate again.
          ))))
    (:then_1
      (OptionData|RefreshPropertyOverrides (Variables|Default|GetToyOptionsComponent)) ;    Now we've converted our variables,\nUpdate save data for device from \n  those new variables.
    )))

(fn ReserveSpawnerForPlayer
  (params (PlayerState :type Player%20State%20Object%20Reference%20%28by%20ref%29))
  (bind _fortminigameprogress (Variables|Default|GetFortMinigameProgress))
  (bind _currentminigame (Variables|/Script/FortniteGame.FortMinigameLogicComponent|GetCurrentMinigame _fortminigameprogress))
  (bind _forcedrespawntime (Variables|/Script/FortniteGame.FortMinigame|GetForcedRespawnTime _currentminigame))
  (bind _asfort_player_start_creative (Utilities|Casting|CastToFortPlayerStartCreative (Variables|/Script/Engine.ChildActorComponent|GetChildActor (Variables|Default|GetPlayerStartCreative)))
    (:then
      (if (Class|FortPlayerStartCreative|CanClaim _asfort_player_start_creative (PlayerState|GetPlayerController (Variables|GetPlayerState)) false)
        (Variables|/Script/FortniteGame.FortPlayerStartCreative|SetLockedPlayerNameSubstring (PlayerState|GetPlayerName (Variables|GetPlayerState)) _asfort_player_start_creative)
        (Variables|/Script/FortniteGame.FortPlayerStartCreative|SetPriorityGroup (Variables|Default|GetHighestPriority) _asfort_player_start_creative)
        (Variables|Default|SetRespawnTimeDelay _forcedrespawntime)
        ;[1 1 1] avoids reserving a spawn pad indefinitely
        (bind _returnvalue (Utilities|Time|SetTimerbyEvent (EventDispatchers|CreateEvent :!event ClearReserveSpawn) (+ _forcedrespawntime (Variables|Default|GetRespawnTimeDelta))))
        (Variables|Default|SetReservedSpawnPadTimer _returnvalue)
        (bind _output_get (Variables|Default|SetbIsReserved true))
        (return _output_get)
        (else
          _)))
    (:CastFailed
      _))
  (return false))

(fn SetSpawnerRotation
  (bind (_returnvalue_x _returnvalue_y) (Math|Vector|BreakVector (Transformation|GetForwardVector (Variables|Default|GetForwardDirection))))
  (bind _playerstartcreative (Variables|Default|GetPlayerStartCreative))
  (bind _childactor (Variables|/Script/Engine.ChildActorComponent|GetChildActor _playerstartcreative))
  (bind _output_get (LocalVariables|SetlocalRot (Math|Rotator|MakeRotfromXZ (Math|Vector|MakeVector _returnvalue_x _returnvalue_y) (0.0 0.0 1.0))))
  (Transformation|SetWorldRotation _playerstartcreative _output_get false false)
  (Transformation|SetMobility (Variables|/Script/Engine.Actor|GetRootComponent _childactor) "Movable")
  (Transformation|SetActorRotation _childactor (LocalVariables|GetlocalRot) false))

(fn ShouldInvisiblePropsBeAllowed
  (return false))

(fn SpawnPlayer_Verse
  (params (Instigator :type Controller%20Object%20Reference))
  (CallFunction|SpawnPlayerEvent Instigator))

(fn TranslateTeamValueToMaterialScoreParam
  (bind (_teamtype _teamindex) (Utilities|Struct|BreakCreativeTeamOption (Variables|UserOptions|Basic|GetPlayerTeam)))
  (if (Creative|TeamOption|IsTeamType (Variables|UserOptions|Basic|GetPlayerTeam) "None")
    (return -1.0)
    (elif (Creative|TeamOption|IsTeamType (Variables|UserOptions|Basic|GetPlayerTeam) "Any")
      (return 0.0)
      (else
        (return (Math|Conversions|ToFloat(Integer) _teamindex))))))

(fn TransmitOnPlayerSpawned
  (params (PlayerState :type Fort%20Player%20State%20Object%20Reference))
  (Networking|SwitchHasAuthority
    (:Authority
      (bind _playerstate (Variables|GetPlayerState)
        (:then
          (if (Utilities|IsValid _playerstate :!class /Script/Engine.KismetSystemLibrary)
            (LocalVariables|SetTransmitterPlayerState (Variables|GetPlayerState))
            (if (and (CallFunction|IsClaimedBy (Variables|GetPlayerState)) (CallFunction|IsInsideSameVolume (LocalVariables|GetTransmitterPlayerState)))
              (bind _ascontroller (Utilities|Casting|CastToController (Actor|GetOwner (Variables|GetPlayerState))))
              (GameplayEvent|BroadcastEvent (Variables|UserOptions|Events|GetOn%20Player%20Spawned) 0 (Utilities|Struct|MakeGameplayEventSpawned _ascontroller))
              (Variables|Default|SetControllerTransmitingOnGameStart 0)
              (if (Variables|Default|GetbIsReserved)
                (CallFunction|ClearReserveSpawn)))))
        (:else)))))

(fn TransmitOnSpawnFailed
  (params (PlayerController :type Player%20Controller%20Object%20Reference))
  (GameplayEvent|BroadcastEvent (Variables|UserOptions|Events|GetOnSpawnFailed) 0 (Utilities|Struct|MakeGameplayEventonSpawnFailed PlayerController))
  (CallFunction|ClearReserveSpawn))

(fn UnbindEventToMinigamePlayerSpawn
  (Networking|SwitchHasAuthority
    (:Authority
      (bind _currentminigame (Variables|/Script/FortniteGame.FortMinigameLogicComponent|GetCurrentMinigame (Variables|Default|GetFortMinigameProgress))
        (:then
          (Minigame|UnbindEventfromOnMinigamePlayerorBotSpawned _currentminigame (EventDispatchers|CreateEvent :!event OnMinigamePlayerSpawned)))
        (:else)))))

(fn UnbindEventToMinigameTeamConfigUpdated
  (Networking|CanExecuteCosmeticEvents
    (:True
      (bind _currentminigame (Variables|/Script/FortniteGame.FortMinigameLogicComponent|GetCurrentMinigame (Variables|Default|GetFortMinigameProgress))
        (:then
          (Minigame|UnbindEventfromMinigameTeamConfigUpdated _currentminigame (EventDispatchers|CreateEvent :!event OnMinigameTeamConfigUpdated)))
        (:else)))))

(fn UpdateAudio
  (bind _audio_loop (Variables|Default|GetAudio%20Loop))
  ;[0.9762 1 0 0.25] Plays audio based on user option. Can either be On, Off or only when Visible
  (Networking|CanExecuteCosmeticEvents
    (:True
      (switch Utilities|FlowControl|Switch|SwitchonEPlayAudioType (Variables|UserOptions|GetPlayAudio)
        (:NewEnumerator0
          _)
        (:NewEnumerator1
          (Audio|Components|Audio|Play _audio_loop 0.0))
        (:NewEnumerator2
          (if (Visibility|ShouldbeVisible (Variables|Default|GetVisibleInGameComponent))
            _
            (else
              _))))
      (Audio|Components|Audio|Stop _audio_loop))))

(fn UpdateEnemyRangeCheck
  (bind _asfort_player_start_creative (Utilities|Casting|CastToFortPlayerStartCreative (Variables|/Script/Engine.ChildActorComponent|GetChildActor (Variables|Default|GetPlayerStartCreative))))
  (Variables|/Script/FortniteGame.FortPlayerStartCreative|SetEnemyRangeCheck (Variables|UserOptions|GetEnemyRangeCheck) _asfort_player_start_creative))

(fn UpdateEnemyRangeOpacity
  (params (Opacity :type Float%20%28double-precision%29))
  (Networking|CanExecuteCosmeticEvents
    (:True
      (Rendering|Material|SetScalarParameterValue (Material|GetMID (Variables|Default|GetEnemyRangeVisualization)) "Opacity" Opacity :!class /Script/Engine.MaterialInstanceDynamic))))

(fn UpdateHologramVisuals
  (bind (_base_color _light_color) (Teams|TeamOption|GetTeamColorforTeamOption self (Variables|UserOptions|Basic|GetPlayerTeam)))
  (bind _isvisible (LocalVariables|GetIsVisible))
  (bind _disabledemissivemult (LocalVariables|GetDisabledEmissiveMult))
  (bind _localcolor (LocalVariables|GetlocalColor))
  (bind _returnvalue (* _localcolor _disabledemissivemult))
  (bind _returnvalue_2 (Math|Color|SelectColor _localcolor _returnvalue _isvisible))
  (bind _mid_spawnplate (Variables|EmissiveShift|GetMID%20SpawnPlate))
  (Networking|CanExecuteCosmeticEvents
    (:True
      (LocalVariables|SetIsVisible (and (Class|CreativeEnabledComponent|IsEnabled (Variables|Default|GetCreative_Enabled_Component)) (not (Creative|TeamOption|IsTeamType (Variables|UserOptions|Basic|GetPlayerTeam) "None"))))
      (bind _output_get (LocalVariables|SetlocalColor (Math|Color|SelectColor (Variables|Default|GetPreviewColor) (Math|Color|SelectColor _light_color (Variables|Default|GetDisabledColor) (LocalVariables|GetIsVisible)) (Building|IsPreviewBuildingActor))))
      (Material|ReattachOrCreateMID (Variables|Debug|GetStaticMeshComponent) _mid_spawnplate)
      (Rendering|Material|SetVectorParameterValue _mid_spawnplate "Emissive" (Math|Color|SelectColor (* _output_get (Variables|Default|GetBaseEmissiveMultiplier)) (* _output_get (LocalVariables|GetDisabledEmissiveMult)) (LocalVariables|GetIsVisible)) :!class /Script/Engine.MaterialInstanceDynamic)
      (bind _hologrammid (CallFunction|GetHologramMID))
      (Rendering|Material|SetVectorParameterValue _hologrammid "BottomColor" _returnvalue_2 :!class /Script/Engine.MaterialInstanceDynamic)
      (Rendering|Material|SetVectorParameterValue _hologrammid "BottomTint" _returnvalue_2 :!class /Script/Engine.MaterialInstanceDynamic)
      (Rendering|Material|SetVectorParameterValue _hologrammid "TopColor" _returnvalue_2 :!class /Script/Engine.MaterialInstanceDynamic)
      (Rendering|Material|SetScalarParameterValue _hologrammid "FadeDistance" (Math|Float|SelectFloat 0.0 (Math|Float|SelectFloat (Variables|Default|GetHologramFadeDistanceEditor) (Variables|Default|GetHologramFadeDistanceGame) (Editor|InEditor)) (Building|IsPreviewBuildingActor)) :!class /Script/Engine.MaterialInstanceDynamic) ; In editor, we care about camera intersecting.\nIn Game, we care about character intersecting.\nIn Preview, we don't want any intersection.
    )))

(fn UpdateClassSetting
  (bind _asfort_player_start_creative (Utilities|Casting|CastToFortPlayerStartCreative (Variables|/Script/Engine.ChildActorComponent|GetChildActor (Variables|Default|GetPlayerStartCreative))))
  (Variables|/Script/FortniteGame.FortPlayerStartCreative|SetApplicableClass (Creative|ClassOption|GetLegacyClassIndex (Variables|UserOptions|GetPlayerClass)) _asfort_player_start_creative))

(fn UpdateIsEnabled
  (params (bIsEnabled :type Boolean))
  (bind _asfort_player_start_creative (Utilities|Casting|CastToFortPlayerStartCreative (Variables|/Script/Engine.ChildActorComponent|GetChildActor (Variables|Default|GetPlayerStartCreative))))
  (Utilities|IsLocationRelevanttoMe (Transformation|GetWorldLocation (Variables|Default|GetBoundsCheckSphere))
    (:IsRelevant
      (Variables|/Script/FortniteGame.FortPlayerStartCreative|SetbIsEnabled bIsEnabled _asfort_player_start_creative))
    (:NotRelevant
      (Variables|/Script/FortniteGame.FortPlayerStartCreative|SetbIsEnabled false _asfort_player_start_creative))))

(fn UpdatePriorityGroup
  (bind _playerstartcreative (Variables|Default|GetPlayerStartCreative))
  (bind _childactor (Variables|/Script/Engine.ChildActorComponent|GetChildActor _playerstartcreative))
  (if (Variables|UserOptions|GetPriorityGroup_Override)
    (bind _asfort_player_start_creative (Utilities|Casting|CastToFortPlayerStartCreative _childactor))
    (Variables|/Script/FortniteGame.FortPlayerStartCreative|SetPriorityGroup (Variables|UserOptions|GetPriorityGroup) _asfort_player_start_creative)
    (else
      (bind _asfort_player_start_creative_2 (Utilities|Casting|CastToFortPlayerStartCreative _childactor))
      (Variables|/Script/FortniteGame.FortPlayerStartCreative|SetPriorityGroup (Class|GetClassDefaults "/Script/FortniteGame.FortPlayerStartCreative") _asfort_player_start_creative_2))))

(fn UpdateTeamSetting
  (bind _playerstartcreative (Variables|Default|GetPlayerStartCreative))
  (bind _childactor (Variables|/Script/Engine.ChildActorComponent|GetChildActor _playerstartcreative))
  (Utilities|PlayerTeamifEnabled
    (:NoneTeam
      (bind _asfort_player_start_creative (Utilities|Casting|CastToFortPlayerStartCreative _childactor))
      (Variables|/Script/FortniteGame.FortPlayerStartCreative|SetApplicableTeam (Utilities|Enum|LiteralenumEFortTeam "Spectator") _asfort_player_start_creative))
    (:Any
      (bind _asfort_player_start_creative_2 (Utilities|Casting|CastToFortPlayerStartCreative _childactor))
      (Variables|/Script/FortniteGame.FortPlayerStartCreative|SetApplicableTeam (Utilities|Enum|LiteralenumEFortTeam "HumanCampaign") _asfort_player_start_creative_2))
    (:ValidTeam
      (bind _asfort_player_start_creative_3 (Utilities|Casting|CastToFortPlayerStartCreative _childactor))
      (Variables|/Script/FortniteGame.FortPlayerStartCreative|SetApplicableTeam (Creative|TeamOption|ConverttoTeam (Variables|UserOptions|Basic|GetPlayerTeam)) _asfort_player_start_creative_3))))

(fn UpdateUseAsIslandStart
  (Utilities|IsLocationRelevanttoMe (Transformation|GetWorldLocation (Variables|Default|GetBoundsCheckSphere))
    (:IsRelevant
      _)
    (:NotRelevant
      ;[1 0.7531 0.6208] IsLocationRelevantToMe is always not relevant in Valkyrie_PIECreativeToolsOnlyMode since there's no volume. So ignore the IsLocationRelevant result if we're in that mode.
      (bind _returnvalue (Development|GetConsoleVariableBoolValue "EditPlayspace.StartWithSavingAllowed"))
      (if (not _returnvalue)
        (Variables|UserOptions|Basic|SetUseAsIslandStart false))))
  (bind _asfort_player_start_creative (Utilities|Casting|CastToFortPlayerStartCreative (Variables|/Script/Engine.ChildActorComponent|GetChildActor (Variables|Default|GetPlayerStartCreative))))
  (bind _canspawnincreative (CallFunction|GetCanSpawnPlayer))
  (Variables|/Script/FortniteGame.FortPlayerStartCreative|SetbIsEnabled _canspawnincreative _asfort_player_start_creative)
  (Variables|/Script/FortniteGame.FortPlayerStartCreative|SetbUseAsIslandStart (select _canspawnincreative (Variables|UserOptions|Basic|GetUseAsIslandStart) false) _asfort_player_start_creative))

(fn ConstructionScript
  (Utilities|FlowControl|Sequence
    (:then_0
      (if (Editor|InEditor)
        (CallFunction|InitializeDevice)
        (Editor|SetIgnoreBoundsforEditorFocus (Variables|Default|GetEnemyRangeVisualization) true)))
    (:then_1
      (|CallParentFunction))))

(fn ValidateAndCacheTransmitData
  (params (FortPlayerState :type Fort%20Player%20State%20Object%20Reference))
  (if (Utilities|IsValid FortPlayerState :!class /Script/Engine.KismetSystemLibrary)
    (bind _currentminigame (Variables|/Script/FortniteGame.FortMinigameLogicComponent|GetCurrentMinigame (Variables|Default|GetFortMinigameProgress))
      (:then
        (if (Minigame|IsinProgress _currentminigame)
          (if (CallFunction|IsClaimedBy FortPlayerState)
            (Variables|Default|SetControllerTransmitingOnGameStart FortPlayerState))))
      (:else))))
