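# Translucent unlit material for a player spawner's team display, drawing a two-digit score as colored dots.
# Score picks tens and ones digits by offsetting UVs into a number-strip texture, split at the quad's middle.
# The digits mask a tiled circle texture tinted by Color, and Win blends in a sine pulse over time.
# An If blanks the result without a score, with OpacityMultiplier and camera depth fade scaling it.
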Material /CRD_PlayerSpawn/Materials/PlayerSpawnerTeamDisplay
graph {
 properties {blendMode=BLEND_Translucent, pixelDepthOffsetMode=PDOM_Legacy, shadingModel=MSM_Unlit, usedShadingModels=Unlit}
 nodes [
  Multiply(A=Frac(None=Time())) {constB=10}
  n1 = VectorParameter() {parameterName=Color, defaultValue=<rgba 1 1 1 1>}
  n2 = TextureCoordinate() {uTiling=20, vTiling=14}
  n3 = TextureCoordinate() {uTiling=0.2}
  n4 = ComponentMask(None=@n3) {r=true}
  n5 = ScalarParameter() {defaultValue=1, parameterName=Score}
  n6 = Multiply(A=Floor(None=Multiply(A=@n5) {constB=0.1})) {constB=0.1}
  n7 = ComponentMask(None=@n3) {g=true}
  n8 = TextureSample(UVs=AppendVector(A=Add(A=@n4, B=@n6), B=@n7)) {texture=/Game/Environments/Props/Street/Textures/numbers}
  n9 = Add(A=@n4, B=Multiply(A=Floor(None=@n5)) {constB=0.1})
  n10 = TextureSample(UVs=AppendVector(A=Add(A=@n9) {constB=-0.1}, B=@n7)) {texture=&numbers}
  n11 = TextureCoordinate()
  n12 = Ceil(None=ComponentMask(None=Subtract(A=@n11) {constB=0.5}) {r=true})
  n13 = Multiply(
   A=@n1.RGB
   B=Multiply(
    A=TextureSample(UVs=@n2) {texture=/Game/Environments/Props/Street/Textures/CircleTexture}.R
    B=LinearInterpolate(A=@n8.R, B=@n10.R, Alpha=@n12)
   )
  )
  n14 = Multiply(A=Time()) {constB=2}
  n15 = LinearInterpolate(A=@n13, B=Multiply(A=@n13, B=Sine(None=@n14)), Alpha=ScalarParameter() {parameterName=Win})
  n16 = ScalarParameter() {defaultValue=1, parameterName=OpacityMultiplier, sliderMax=1}
  n17 = Constant()
  Multiply(
   A=If(A=@n5, "A > B"=Multiply(A=@n15, B=@n16), "A == B"=@n17, "A < B"=@n17) {desc=hide if no score}
   B=MaterialFunctionCall(Fade Length=Constant() {r=20}, Fade Offset=ScalarParameter() {defaultValue=320, parameterName=FadeDistance}) {materialFunction=&CameraDepthFade}.Result
  )
  StaticSwitchParameter(True=@n1.RGB, False=Constant()) {defaultValue=true, group=Color, parameterName=Use Color}
 ]
}
